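; Looping gameplay cue notify for the DentureSay dash that starts the air-dash visual effect on the affected player pawn.
; On application it attaches the NS_DentureSay_AirDash Niagara system to the pawn mesh's root at an offset, initially inactive.
; It sets the LocalPlayer parameter from whether the pawn is locally viewed or controlled, then activates the system.
; Debug prints remain, velocity-based orientation is only a note, and no removal logic is present.
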(event EventOnApplication(EitherVFX-preferthisone)
  (params (MyTarget :type Actor%20Object%20Reference) (Parameters :type Gameplay%20Cue%20Parameters%20Structure%20%28by%20ref%29) (ParticleComponents :type Array%20of%20FXSystem%20Component%20Object%20References) (AudioComponents :type Array%20of%20Audio%20Component%20Object%20References) (BurstCameraShakeInstance :type Camera%20Shake%20Base%20Object%20Reference) (BurstDecalInstance :type Decal%20Actor%20Object%20Reference))
  ; You should be able to use this actor's velocity to orient the particle effect the right way
  (Development|PrintString 0 "DashActive")
  (bind _asfort_player_pawn (Utilities|Casting|CastToFortPlayerPawn MyTarget))
  (bind _returnvalue (Niagara|SpawnSystemAttached "/DentureSay/Assets/FX/NS_DentureSay_AirDash.NS_DentureSay_AirDash" (Variables|/Script/Engine.Character|GetMesh _asfort_player_pawn) "root" "0.000000,100.000000,120.000000" "0, 0, 0" "SnapToTarget" false false))
  (Development|PrintString 0 (Utilities|String|ToString(Boolean) (or (Pawn|IsLocallyViewed _asfort_player_pawn) (Pawn|IsLocallyControlled _asfort_player_pawn))))
  (Effects|Components|ParticleSystem|SetBoolParameter _returnvalue "LocalPlayer" (or (Pawn|IsLocallyViewed _asfort_player_pawn) (Pawn|IsLocallyControlled _asfort_player_pawn)))
  (Components|Activation|SetActive _returnvalue true true))

(fn ConstructionScript)
