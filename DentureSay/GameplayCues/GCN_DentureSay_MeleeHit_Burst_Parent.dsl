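; Parent burst GameplayCue for DentureSay melee hits that plays an impact sound based on what was struck.
; On burst it runs the parent handler, then stores the hit location from the effect context's hit result.
; A pawn hit plays the local or enemy player sound, depending on whether the attacker (MyTarget) is locally viewed.
; Any other hit plays the world sound, with all three sounds read from variables and played at that location.
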(event EventOnBurst(EitherVFX-preferthisversion)
  (params (MyTarget :type Actor%20Object%20Reference) (Parameters :type Gameplay%20Cue%20Parameters%20Structure%20%28by%20ref%29) (ParticleComponents :type Array%20of%20FXSystem%20Component%20Object%20References) (AudioComponents :type Array%20of%20Audio%20Component%20Object%20References) (BurstCameraShakeInstance :type Camera%20Shake%20Base%20Object%20Reference) (BurstDecalInstance :type Decal%20Actor%20Object%20Reference))
  (bind (_normalizedmagnitude _rawmagnitude _effectcontext) (Ability|GameplayCue|BreakGameplayCueParameters Parameters))
  (bind _returnvalue (Ability|EffectContext|GetHitResult _effectcontext))
  (bind (_bblockinghit _binitialoverlap _time _distance _location _impactpoint _normal _impactnormal _physmat _hitactor) (Collision|BreakHitResult _returnvalue))
  ;[0.01122 0.15 0.1408 0.5] OnBurst: MyTarget is the attacking player, and HitResult contains what the player attacked.
  (|CallParentFunction MyTarget Parameters ParticleComponents AudioComponents BurstCameraShakeInstance BurstDecalInstance)
  (Variables|Default|SetHit%20Location _location)
  (Utilities|Casting|CastToFortPawn _hitactor
    (:then
      (bind _asfort_pawn (Utilities|Casting|CastToFortPawn MyTarget))
      (if (Pawn|IsLocallyViewed _asfort_pawn)
        (CallFunction|PlayLocalPlayerHitSound)
        (else
          (CallFunction|PlayEnemyPlayerHitSound))))
    (:CastFailed
      (CallFunction|PlayHitWorldSound))))

(fn PlayEnemyPlayerHitSound
  (Audio|PlaySoundatLocation 0 (Variables|Default|GetEnemyPlayerHitSound) (Variables|Default|GetHit%20Location))
  (return))

(fn PlayHitWorldSound
  (Audio|PlaySoundatLocation 0 (Variables|Default|GetHitWorldSound) (Variables|Default|GetHit%20Location))
  (return))

(fn PlayLocalPlayerHitSound
  (Audio|PlaySoundatLocation 0 (Variables|Default|GetLocalPlayerHitSound) (Variables|Default|GetHit%20Location))
  (return))
