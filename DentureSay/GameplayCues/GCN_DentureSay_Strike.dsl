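; GCN_DentureSay_Strike is a burst gameplay cue notify that configures the Niagara effect for a strike.
; After the parent OnBurst, it validates a spawned Niagara component and feeds it cue parameters as variables.
; Location becomes BottomPosition, and RawMagnitude doubles as a Z offset for TopPosition.
; A non-zero Normal sets IsGrounded, and the Fort.SurfaceType.Water tag among aggregated target tags sets HitWater.
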(fn OnBurst(Niagara)
  (params (MyTarget :type Actor%20Object%20Reference) (Parameters :type Gameplay%20Cue%20Parameters%20Structure%20%28by%20ref%29) (ParticleComponents :type Array%20of%20Niagara%20Particle%20System%20Component%20Object%20References) (AudioComponents :type Array%20of%20Audio%20Component%20Object%20References) (BurstCameraShakeInstance :type Camera%20Shake%20Base%20Object%20Reference) (BurstDecalInstance :type Decal%20Actor%20Object%20Reference))
  (bind (_normalizedmagnitude _rawmagnitude _effectcontext _matchedtagname _originaltag _aggregatedsourcetags _aggregatedtargettags _location _normal) (Ability|GameplayCue|BreakGameplayCueParameters Parameters))
  (bind _output (Utilities|Array|Get(acopy) ParticleComponents))
  (|CallParentFunction MyTarget Parameters ParticleComponents AudioComponents BurstCameraShakeInstance BurstDecalInstance)
  (Utilities|FlowControl|Sequence
    (:then_0
      (if (Utilities|Array|IsValidIndex ParticleComponents)
        (Utilities|IsValid _output
          (:Is%20Valid
            (Niagara|SetNiagaraVariable(Position) _output "BottomPosition" _location)
            (Niagara|SetNiagaraVariable(Position) _output "TopPosition" (Utilities|Operators|Add _location (Math|Vector|MakeVector 0.0 0.0 _rawmagnitude)))
            (Niagara|SetNiagaraVariable(Bool) _output "IsGrounded" (not (Math|Vector|VectorIsZero _normal)))
            (Niagara|SetNiagaraVariable(Bool) _output "HitWater" (GameplayTags|HasTag _aggregatedtargettags "(TagName=\"Fort.SurfaceType.Water\")"))))))
    (:then_1
      (return))))
